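; Event-triggered gameplay ability that removes the Phone Booth disguise from the avatar.
; Activation applies parts-swap and attribute-cost effects, plays a GameplayCue and forces a net update, then after a brief delay force-removes the AGID item from the player's inventory.
; ShouldAbilityRespondToEvent accepts events by default, ignores weakspot damage, and accepts damage dealt only if it is nonzero and hits a player pawn.
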(event Ability|EventActivateAbilityFromEvent
  (params (EventData :type Gameplay%20Event%20Data%20Structure%20%28by%20ref%29))
  (Ability|ApplyGameplayEffectToOwner (Variables|Default|GetGE_PartsSwap))
  ;[1 1 1] Apply Cost
  (Ability|ApplyGameplayEffectToOwner (Variables|Default|GetGE_AttributeCost))
  ;[1 1 1] Apply GameplayCue
  (Ability|ExecuteGameplayCueOnOwner (Variables|Default|GetGC))
  (Networking|ForceNetUpdate (Ability|GetAvatarActorfromActorInfo))
  (Utilities|FlowControl|Delay 0.01) ; FrameBeforeRemoval
  ;[1 1 1] Remove the Character Parts.
  (bind _agid (Variables|Default|GetAGID)
    (:then
      (Inventory|RemoveItemFromPlayer :PlayerController (Fortnite|GetFortPlayerControllerfromActor (Ability|GetAvatarActorfromActorInfo)) :ItemDefinition _agid :AmountToRemove -1 :bForceRemoval true :bSkipVerbRemoval false))
    (:else)))

(fn ShouldAbilityRespondToEvent
  (params (ActorInfo :type Gameplay%20Ability%20Actor%20Info%20Structure) (Payload :type Gameplay%20Event%20Data%20Structure))
  (bind (_eventtag _instigator _target _optionalobject _optionalobject2 _contexthandle _instigatortags _targettags _eventmagnitude) (Utilities|Struct|BreakGameplayEventData Payload))
  ;[1 1 1] If the event was the damage event check if it is a player pawn
  (switch Utilities|FlowControl|Switch|SwitchonGameplayTag _eventtag
    (:Default
      (return true))
    (:Event.Damage.Dealt.Weakspot
      (return false))
    (:Event.Damage.Dealt
      (if (== _eventmagnitude 0.0)
        (return false)
        (else
          (Utilities|Casting|CastToFortPlayerPawn _target
            (:then
              (return true))
            (:CastFailed
              (return false))))))))
