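; GA_Random_Outfit_WildCard is a Gameplay Ability Blueprint, filed with the phone booth station's disguise assets and named for a wildcard random outfit.
; Its graph adds no logic of its own: the activate event holds only a disabled parent-call placeholder.
; The end-ability event calls the parent implementation, passing false for the cancelled flag.
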(event Ability|EventActivateAbility
  (|CallParentFunction) ; This node is disabled and will not be called.\nDrag off pins to build functionality.
)

(event Ability|EventOnEndAbility
  (params (bWasCancelled :type Boolean))
  (|CallParentFunction false))
